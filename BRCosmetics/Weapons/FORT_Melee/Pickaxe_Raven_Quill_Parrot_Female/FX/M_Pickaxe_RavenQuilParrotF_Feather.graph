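# Masked, two-sided lit material for feather Niagara mesh particles on the female Raven Quill Parrot pickaxe.
# Particle dynamic parameters OffsetX and OffsetY snap UVs to one of four regions in borrowed Reaper glider textures.
# Erode drives a smoke-noise alpha dissolve, while CurlIntesity scales animated, actor-relative vector noise.
# Fresnel-driven thin-film iridescence, from a curve atlas or rainbow texture, adds to diffuse by roughness, tinted by particle color.
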Material /BRCosmetics/Weapons/FORT_Melee/Pickaxe_Raven_Quill_Parrot_Female/FX/M_Pickaxe_RavenQuilParrotF_Feather
graph {
 properties {bUsedWithNiagaraMeshParticles=true, blendMode=BLEND_Masked, pixelDepthOffsetMode=PDOM_Legacy, twoSided=true, usedShadingModels=Default Lit}
 nodes [
  n1 = Constant() {r=0.08}
  n2 = Constant() {r=-0.05}
  n3 = DynamicParameter() {defaultValue=<rgba 0.9 0.9 1 0>, paramNames [OffsetX,Erode,CurlIntesity,OffsetY]}
  n4 = Clamp(None=Floor(None=Multiply(B=@n3.OffsetX) {constA=2}))
  n5 = Constant()
  n6 = Constant() {r=0.55}
  n7 = Clamp(None=Floor(None=Multiply(B=@n3.OffsetY) {constA=2}))
  n8 = Add(
   A=TextureCoordinate() {uTiling=1.1, vTiling=0.9}
   B=AppendVector(A=LinearInterpolate(A=@n1, B=@n2, Alpha=@n4), B=LinearInterpolate(A=@n5, B=@n6, Alpha=@n7))
  )
  n9 = TextureSampleParameter2D(UVs=@n8) {parameterName=Diffuse, texture=/Game/Gadgets/Assets/VinderTech_GliderChute/Glider_Reaper/Textures/Glider_Variant_028_2_d}
  n10 = Multiply(
   A=TextureSample(UVs=TextureCoordinate() {uTiling=2, vTiling=2}) {texture=&T_Inky_Smoke_Tile}.G
   B=@n9.A
  )
  n11 = DynamicParameter() {defaultValue=<rgba 0.9 0.9 1 0>, paramNames [OffsetX,Erode,CurlIntesity,OffsetY]}
  Multiply(
   A=ParticleColor().A
   B=MaterialFunctionCall!(Dynamic Param=LinearInterpolate(A=@n10, B=@n9.A, Alpha=@n11.Erode), Alpha=@n10) {materialFunction=/Game/Effects/Fort_Effects/Materials/_Functions/Erode_Alpha}.Result
  )
  n12 = Multiply(
   A=ScalarParameter() {defaultValue=1, group=ThinFilm, parameterName=ThinFilm_Scale}
   B=Fresnel(ExponentIn=ScalarParameter() {defaultValue=0.5, group=ThinFilm, parameterName=ThinFilm_EX}, Normal=VertexNormalWS()) {baseReflectFraction=0}
  )
  n13 = CurveAtlasRowParameter(CurveTime=@n12) {
   atlas=&AthenaCurveAtlas01
   curve=/Game/Athena/LinearColorCurves/LuchadorVFX_Glider
   defaultValue=7
   parameterName=Curve
  }
  n14 = Multiply(
   A=StaticSwitchParameter(True=@n13, False=TextureSample(UVs=@n12) {texture=/Game/Global/Textures/Default/Masks_and_Gradients/FortniteRainbow}.RGB) {parameterName=Use Curve for Thin Film}
   B=ScalarParameter() {defaultValue=1, group=ThinFilm, parameterName=ThinFilm_Intensity}
  )
  n15 = ScalarParameter() {defaultValue=0.25, parameterName=Roughness}
  n16 = Clamp(None=Multiply(A=OneMinus(None=@n15), B=ScalarParameter() {defaultValue=1, group=ThinFilm, parameterName=ThinFilm_RoughnessScale}))
  n17 = Fresnel(ExponentIn=ScalarParameter() {defaultValue=1.2164, group=ThinFilm, parameterName=ThinFilm_Visibility}) {baseReflectFraction=0}
  Multiply(A=LinearInterpolate(A=@n9.RGB, B=Add(A=@n9.RGB, B=@n14), Alpha=Multiply(A=@n16, B=@n17)), B=ParticleColor().RGB)
  n18 = Subtract(A=Subtract(A=WorldPosition().XYZ, B=ActorPositionWS()), B=Multiply(A=ScalarParameter() {defaultValue=5, parameterName=CurlSpeed}, B=Time()))
  n19 = ScalarParameter() {defaultValue=100, parameterName=CurlScale}
  Multiply(
   A=DynamicParameter() {defaultValue=<rgba 0.9 0.9 1 0>, paramNames [OffsetX,Erode,CurlIntesity,OffsetY]}.CurlIntesity
   B=VectorNoise(World Position=Divide(A=@n18, B=@n19)) {noiseFunction=VNF_VectorALU}
  )
  ScalarParameter() {defaultValue=0.25, parameterName=Specular}
  ScalarParameter() {defaultValue=0.25, parameterName=Roughness}
  TextureSampleParameter2D(UVs=@n8) {
   parameterName=Normal
   samplerType=SAMPLERTYPE_Normal
   texture=/Game/Gadgets/Assets/VinderTech_GliderChute/Glider_Reaper/Textures/Glider_Variant_028_2_n
  }
  VectorParameter() {parameterName=BaseColor, defaultValue=<rgba 0.0333 0.0747 0.2812 1>}
 ]
}
