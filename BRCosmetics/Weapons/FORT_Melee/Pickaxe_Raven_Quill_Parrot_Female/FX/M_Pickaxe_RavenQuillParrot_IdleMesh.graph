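# Translucent, unlit, two-sided animated flame material for the idle effect of the Raven Quill Parrot Female pickaxe.
# Flow-map-distorted panning noise and a gradient form flames, color-remapped through a lookup texture and HDR-boosted; IS_Ribbon switches UV orientation.
# Noise and sine push vertices along normals, masked by vertex blue.
# Opacity combines the flame with gradient, inverted sphere mask and sine, with depth fade skipped on mobile.
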Material /BRCosmetics/Weapons/FORT_Melee/Pickaxe_Raven_Quill_Parrot_Female/FX/M_Pickaxe_RavenQuillParrot_IdleMesh
graph {
 properties {
  bUsedWithMeshDeformer=true
  bUsedWithNiagaraMeshParticles=true
  bUsedWithNiagaraRibbons=true
  bUsedWithNiagaraSprites=true
  bUsedWithSkeletalMesh=true
  blendMode=BLEND_Translucent
  pixelDepthOffsetMode=PDOM_Legacy
  shadingModel=MSM_Unlit
  twoSided=true
  usedShadingModels=Unlit
 }
 nodes [
  n1 = ComponentMask(None=VectorParameter() {group=FlameShape, parameterName=Flames_UVScale, defaultValue=<rgba 0.6 0.4 0 0>}.RGB) {g=true, r=true}
  n2 = MaterialFunctionCall(Float2=TextureCoordinate() {uTiling=-1, vTiling=4}) {materialFunction=&BreakOutFloat2Components}
  n3 = VectorParameter() {parameterName=DistortUVScale, defaultValue=<rgba 0.5 0.5 0 0>}
  n4 = Add(
   A=Multiply(A=Time(), B=ScalarParameter() {defaultValue=1, parameterName=DistortUVAnim})
   B=Multiply(A=TextureCoordinate(), B=ComponentMask(None=@n3.RGB) {g=true, r=true})
  )
  n5 = Multiply(
   A=ComponentMask(None=TextureSample(UVs=@n4) {samplerType=SAMPLERTYPE_LinearColor, texture=/Game/Effects/Fort_Effects/Creatures/Elemental/Fire/MergedFireFlowMap}.RGB) {g=true, r=true}
   B=ScalarParameter() {defaultValue=0.01, parameterName=Distort}
  ) {constB=0.5}
  n6 = Add(
   A=StaticSwitch(True=AppendVector(A=@n2.G, B=@n2.G), False=TextureCoordinate() {uTiling=4}, Value=StaticBoolParameter() {parameterName=IS_Ribbon})
   B=@n5
  )
  n7 = ScalarParameter() {group=FlameShape, parameterName=Flame_UVOffset}
  n8 = Multiply(A=Time(), B=ScalarParameter() {defaultValue=2, group=FlameShape, parameterName=Flame_Speed})
  n9 = Panner(Coordinate=Add(A=Multiply(A=@n1, B=@n6), B=@n7), Time=@n8) {speedX=0.2, speedY=1}
  n10 = Multiply(A=TextureCoordinate() {vTiling=0.5}, B=@n1)
  n11 = TextureSample(UVs=Panner(Coordinate=Add(A=@n10, B=@n7)) {speedX=-0.5, speedY=1}) {texture=&T_NoisePack_01}
  n12 = Constant() {r=0.2}
  n13 = MaterialFunctionCall(Float2=TextureCoordinate() {uTiling=-1, vTiling=6}) {materialFunction=&BreakOutFloat2Components}
  n14 = StaticSwitch(
   True=AppendVector(A=@n13.G, B=@n13.G)
   False=TextureCoordinate() {uTiling=6, vTiling=1.5}
   Value=StaticBoolParameter() {parameterName=IS_Ribbon}
  )
  n15 = Add(A=@n7, B=Multiply(A=@n1, B=Add(A=@n14, B=@n5)))
  n16 = Multiply(
   A=Add(
    A=MaterialFunctionCall(In=@n11.G, Contrast=@n12) {materialFunction=&CheapContrast}.Result
    B=TextureSample(UVs=@n9) {texture=&T_NoisePack_01}.R
   )
   B=TextureSample(UVs=Panner(Coordinate=@n15, Time=@n8) {speedX=-0.2, speedY=1}) {texture=&T_NoisePack_01}.G
  )
  n17 = TextureCoordinate()
  n18 = ComponentMask(None=@n17) {g=true}
  n19 = ComponentMask(None=@n17) {r=true}
  n20 = MaterialFunctionCall(UV Channel=OneMinus(None=AppendVector(A=@n18, B=@n19))) {materialFunction=&LinearGradient}
  n21 = MaterialFunctionCall() {materialFunction=&LinearGradient}
  n22 = StaticBoolParameter() {parameterName=IS_Ribbon}
  n23 = Multiply(A=Power(
    Base=StaticSwitch(True=@n20.VGradient, False=@n21.VGradient, Value=@n22)
    Exp=ScalarParameter() {defaultValue=2, group=FlameShape, parameterName=FireLine}
   )) {constB=0.5}
  n24 = MaterialFunctionCall(
   In=Power(Base=Add(A=@n16, B=@n23), Exp=ScalarParameter() {defaultValue=1.1, parameterName=Pow}) {constExponent=1}
   Contrast=ScalarParameter() {defaultValue=0.5, group=FlameShape, parameterName=HardEdgeness}
  ) {materialFunction=&CheapContrast}
  n25 = OneMinus(None=VertexColor().G)
  n26 = Multiply(A=@n25, B=@n25)
  n27 = TextureSampleParameter2D(UVs=StaticSwitch(True=@n24.Result, False=Multiply(A=@n26, B=@n24.Result), Value=StaticBoolParameter() {parameterName=IS_Ribbon})) {
   parameterName=ReMappedColorTex
   samplerType=SAMPLERTYPE_LinearColor
   texture=&/T_Pickaxe_RavenQuilParrotF_lookup
  }
  n28 = MaterialFunctionCall(
   LDR Input=Clamp(None=@n27.RGB)
   HDR Tint=Multiply(A=@n27.RGB, B=VectorParameter() {group=FlameShape, parameterName=HDR_Tint, defaultValue=<rgba 1 1 1 0>}.RGB)
   HDR Intensity Multiplier=ScalarParameter() {defaultValue=3, group=FlameShape, parameterName=HDR_Intensity}
  ) {materialFunction=&DeriveHDRfromLDR}
  n29 = ScalarParameter() {defaultValue=10, group=FlameShape, parameterName=MobileBoost}
  MaterialFunctionCall(Emissive=ShadingPathSwitch(Default=@n28.RemappedRange, Deferred=@n28.RemappedRange, Forward=@n28.RemappedRange, Mobile=Multiply(A=@n29, B=@n27.RGB))) {materialFunction=&MF_FX_EmissiveAdjustments}
  n30 = TextureCoordinate() {uTiling=-1}
  n31 = AppendVector(A=ComponentMask(None=@n30) {g=true}, B=ComponentMask(None=@n30) {r=true})
  n32 = TextureCoordinate()
  n33 = StaticBoolParameter() {parameterName=IS_Ribbon}
  n34 = OneMinus(None=ComponentMask(None=StaticSwitch(True=@n31, False=@n32, Value=@n33)) {g=true})
  n35 = TextureSample!(UVs=@n9, Tex, Level, Apply View MipBias) {mipValueMode=TMVM_MipLevel, texture=&T_NoisePack_01}
  n36 = Multiply(
   A=LinearInterpolate(B=ScalarParameter() {defaultValue=8, group=WPO, parameterName=Distort Edge}, Alpha=Multiply(A=@n34, B=@n34))
   B=Multiply(A=LinearInterpolate(Alpha=@n35.G) {constA=-1}, B=VertexNormalWS())
  )
  n37 = TextureCoordinate()
  n38 = Sine(None=Add(A=ComponentMask(None=@n37) {g=true}, B=ScalarParameter() {defaultValue=-0.125, group=WPO, parameterName=Distort_offset})) {period=2}
  n39 = ScalarParameter() {defaultValue=6, group=WPO, parameterName=Distort_power}
  n40 = Multiply(A=VertexNormalWS(), B=Multiply(A=Power(Base=@n38, Exp=@n39), B=ScalarParameter() {defaultValue=1, group=WPO, parameterName=Distort_Multi}))
  Multiply(
   A=OneMinus(None=VertexColor().B)
   B=Multiply(A=Add(A=@n36, B=@n40) {constB=0}, B=ScalarParameter() {defaultValue=1, group=WPO, parameterName=WPO_OverallIntensity})
  )
  n41 = Clamp(None=@n24.Result)
  n42 = Power(
   Base=MaterialFunctionCall() {materialFunction=&LinearGradient}.VGradient
   Exp=ScalarParameter() {defaultValue=4, group=Alpha, parameterName=VGradientContrast_Top}
  )
  n43 = Multiply(
   A=TextureCoordinate()
   B=AppendVector(
    A=ScalarParameter() {defaultValue=0.25, group=Alpha, parameterName=Sphere_U}
    B=ScalarParameter() {defaultValue=1, group=Alpha, parameterName=Sphere_V}
   )
  )
  n44 = ComponentMask(None=VectorParameter() {group=Alpha, parameterName=Spheremaskradius, defaultValue=<rgba 0.125 0.9 0 0>}.RGB) {g=true, r=true}
  n45 = ScalarParameter() {defaultValue=0.85, group=Alpha, parameterName=Radius}
  n46 = ScalarParameter() {defaultValue=0.125, group=Alpha, parameterName=Hardness}
  n47 = Power(
   Base=OneMinus(None=SphereMask(A=@n43, B=@n44, Radius=@n45, Hardness=@n46))
   Exp=ScalarParameter() {defaultValue=2.4503, group=Alpha, parameterName=SphereMaskGradient Contrast}
  )
  n48 = Saturate(None=Multiply(A=Multiply(A=@n42, B=@n47), B=ScalarParameter() {defaultValue=2, group=Alpha, parameterName=Enhance}))
  n49 = Sine(None=Divide(A=TextureCoordinate()))
  n50 = Multiply(
   A=ScalarParameter() {defaultValue=2, group=Alpha, parameterName=MasterAlpha}
   B=Multiply(A=Multiply(A=@n41, B=@n48), B=ComponentMask(None=@n49) {r=true})
  )
  n51 = DepthFade(Opacity=@n50, FadeDistance=ScalarParameter() {defaultValue=2, group=Alpha, parameterName=Fade Distance})
  ShadingPathSwitch(Default=@n51, Deferred=@n51, Mobile=@n50) {desc=Migrated from FeatureLevelSwitch}
 ]
}
